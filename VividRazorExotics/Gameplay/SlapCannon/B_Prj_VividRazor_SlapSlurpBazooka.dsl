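; Slap Cannon slurp bazooka projectile that explodes on stop, bounce or water entry and applies a slap gameplay effect to hit results.
; The explosion handler asks player pawn 0's ability system to apply the effect, then fires a splash cue and fades flight audio.
; Bouncing shots hide on hitting actors, otherwise tumble off walls; the floor pickup event and health-check branches are empty.
; Homing events retune turn speeds near or past the target.
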(event Custom|SpawnSlurpJuice)

(event Custom|Play%20Explosion%20Sound
  ;[1 1 1] Event Play Explosion Sound
  (Utilities|FlowControl|DoOnce false))

(event EventOnStop
  (params (Hit :type Hit%20Result%20Structure%20%28by%20ref%29))
  ;[1 1 1] Event on Stop play Hit Fx
  (Variables|Default|SetHit Hit)
  (Utilities|FlowControl|Sequence
    (:then_0
      (if (not (Variables|Default|GetbHasExploded?))
        (Variables|Default|SetbHasExploded? true)
        (CallFunction|PlayExplosionSound)
        (Delivery|DoExplosion) ; Explosion
        (Delivery|Kill)))
    (:then_1
      (Development|SetHiddeninGame (Variables|Default|GetSK_SlurpJuice) true false)
      (Networking|SwitchHasAuthority
        (:Remote
          (Utilities|FlowControl|Delay (Variables|VFX|GetFX%20Linger%20Time))
          (Actor|DestroyActor))))))

(event Custom|Reset%20Explosion%20Sound
  ;[1 1 1] Event Play Explosion Sound
  (Utilities|FlowControl|DoOnce false))

(event Custom|SpawnedBandageFx
  ;[1 1 1] BounceEvent
  (Audio|SpawnSoundatLocation (Variables|Default|GetSpawnSlurpJuiceSound) (Transformation|GetActorLocation))
  (Audio|Components|Audio|Stop (Variables|Default|GetFlightSound))
  (Components|Movement|StopMovementImmediately (Variables|Movement|GetProjectileMovementComponent))
  (Delivery|Kill))

(event EventOnBounce
  (params (Hit :type Hit%20Result%20Structure%20%28by%20ref%29))
  ;[1 1 1] BounceEvent
  (Variables|Default|SetHit Hit)
  (Utilities|FlowControl|Gate false
    (:Exit
      (Delivery|DoExplosion) ; Explosion
      (Components|Activation|Deactivate (Variables|Default|GetNS_SlurpBazooka_Trail))
      (Rendering|SetVisibility (Variables|Default|GetSK_SlurpJuice) true false) ; Unhide Slurp Juice Mesh
      (Audio|SpawnSoundatLocation (Variables|Default|GetBounceOffWallSound) (Transformation|GetActorLocation))
      (Components|DestroyComponent (Variables|Default|GetSK_SlurpJuice)))))

(event EventBeginPlay
  (Utilities|FlowControl|Sequence
    (:then_0
      (Variables|VFX|SetFX%20Linger%20Time (Math|Random|RandomFloatinRange 1.0 1.0)))
    (:then_1
      (CallFunction|ResetExplosionSound))))

(event Custom|CloseGate
  ;[1 1 1] BounceEvent
  (Utilities|FlowControl|Gate false
    (:Exit
      (Delivery|DoExplosion) ; Explosion
      (Components|Activation|Deactivate (Variables|Default|GetNS_SlurpBazooka_Trail))
      (Rendering|SetVisibility (Variables|Default|GetSK_SlurpJuice) true false) ; Unhide Slurp Juice Mesh
      (Audio|SpawnSoundatLocation (Variables|Default|GetBounceOffWallSound) (Transformation|GetActorLocation))
      (Components|DestroyComponent (Variables|Default|GetSK_SlurpJuice)))))

(event ProjectileMovement|EventOnHomingPassedTarget
  (Projectile|SetHomingTurnSpeedMin 0 0 3.6e+02 :!class /Script/FortniteGame.FortProjectileMovementInterface)
  (Projectile|SetHomingTurnSpeedMax 0 0 7.2e+02 :!class /Script/FortniteGame.FortProjectileMovementInterface)
  (Projectile|SetHomingTimeUntilMaxTurnSpeed 0 0 0.5 :!class /Script/FortniteGame.FortProjectileMovementInterface))

(event ProjectileMovement|EventOnReachedLockTargetDistanceThreshold
  (Projectile|SetHomingTurnSpeedMin 0 0 3.6e+02 :!class /Script/FortniteGame.FortProjectileMovementInterface)
  (Projectile|SetHomingTurnSpeedMax 0 0 7.2e+02 :!class /Script/FortniteGame.FortProjectileMovementInterface)
  (Projectile|SetHomingTimeUntilMaxTurnSpeed 0 0 0.5 :!class /Script/FortniteGame.FortProjectileMovementInterface))

(event EventOnExploded
  (params (HitActors :type Array%20of%20Actor%20Object%20References) (HitResults :type Array%20of%20Hit%20Result%20Structures))
  (bind (_bblockinghit _binitialoverlap _time _distance _location _impactpoint _normal) (Collision|BreakHitResult (Variables|Default|GetHit)))
  (bind _returnvalue (Transformation|GetActorLocation))
  (bind _output (Utilities|Array|Get(acopy) HitActors))
  (bind _hit (Variables|Default|GetHit))
  (SoundIndicator|ShowSoundIndicator (Variables|Default|GetFortSoundIndicator) _returnvalue (Variables|Default|GetBounceOffWallSound) 1.0 0)
  (CallFunction|AddSlapEffect _returnvalue HitActors HitResults)
  (Audio|Components|Audio|FadeOut (Variables|Default|GetFlightSound) 0.3 0.0 "Linear" :!class /Script/Engine.AudioComponent)
  (CallFunction|SpawnExplosionVFX _hit)
  (CallFunction|ApplyWaterinRadiusifCurieEnabled)
  (if (Variables|/Script/Engine.ProjectileMovementComponent|GetbShouldBounce (Variables|Movement|GetProjectileMovementComponent))
    ;[1 1 1] Check if Healing Anyone
    (if (> (Utilities|Array|LastIndex HitActors) -1)
      (bind _asfort_player_pawn (Utilities|Casting|CastToFortPlayerPawn _output)
        (:then
          (if (!= (FortPawn|GetMaxHealth _asfort_player_pawn) 0.0)
            (if (and (== (FortPawn|HealthPercent _asfort_player_pawn) 1.0) (>= (CallFunction|ShieldPercent _asfort_player_pawn) 1.0))
              _
              (else
                _))
            (else
              _)))
        (:CastFailed
          (bind _asnpc_pawn_wildlife_parent (Utilities|Casting|CastToNPC_Pawn_Wildlife_Parent _output))
          (if (and (!= (FortPawn|GetMaxHealth _asnpc_pawn_wildlife_parent) 0.0) (!= (FortPawn|HealthPercent _asnpc_pawn_wildlife_parent) 1.0))
            _)))
      (Variables|Default|SetHitResult _hit)
      (Rendering|SetActorHiddenInGame true)
      (Variables|Default|SetbHasExploded? true)
      (CallFunction|CloseGate)
      (else
        ;[1 1 1] If not healing self, and hitting floor, don't bounce, just spawn pick up and die.
        (if (< (Math|Trig|Acos(Degrees) (Math|Vector|DotProduct _normal (Math|Vector|VectorUp))) 3e+01)
          (CallFunction|SpawnSlurpJuice)
          (else
            ;[1 1 1] Bounce Off Wall
            (Variables|/Script/Engine.ProjectileMovementComponent|SetbRotationFollowsVelocity false (Variables|Movement|GetProjectileMovementComponent))
            (Variables|/Script/Engine.RotatingMovementComponent|SetRotationRate "1000.000000, 0,7.000000" (Variables|Default|GetRotatingMovement))
            (Projectile|SetGravityScale 0.7)
            (Components|Activation|Deactivate (Variables|Default|GetNS_SlurpBazooka_Trail))))))))

(event Custom|SpawnExplosionVFX
  (params (Hit :type Hit%20Result%20Structure%20%28by%20ref%29))
  (bind (_bblockinghit _binitialoverlap _time _distance _location _impactpoint _normal _impactnormal _physmat _hitactor) (Collision|BreakHitResult Hit))
  ;[0.04281 0.3 0 0.3] Decals and fx to spawn
  (GameplayCue|ExecuteGameplayCue%28Burst%29Local _hitactor (Variables|Default|GetExplosion%20Cue%20Parameter) (Ability|GameplayCue|MakeGameplayCueParameters :NormalizedMagnitude 0.0 :RawMagnitude 0.0 :Location _location :Normal _normal :GameplayEffectLevel 1 :AbilityLevel 1 :bReplicateLocationWhenUsingMinimalRepProxy false)) ; Splash gameplay cue
)

(event OnEnterWaterEvent(WaterInteractionComponent)
  (params (WaterBody :type Fort%20Water%20Body%20Actor%20Object%20Reference) (WaterInteractionComponent :type Fort%20Water%20Interaction%20Component%20Object%20Reference) (bIsFirstBody :type Boolean))
  (bind (_waterplanelocation _waterplanenormal _watersurfaceposition _waterdepth _waterbodyidx) (FX|GetLastWaterSurfaceInfo WaterInteractionComponent :!class /Script/FortniteGame.FortWaterInteractionComponent))
  ;[0 0.298 1] Water Impact Projectile
  (if bIsFirstBody
    (Audio|Components|Audio|SetSound (Variables|Default|GetFlightSound) "/Game/Sounds/Fort_Weapons/Launchers/Healing/HealingGrenade_InWaterLoop_Cue.HealingGrenade_InWaterLoop_Cue")
    (bind _returnvalue (Effects|Components|ParticleSystem|SpawnFXAtLocation "/Game/Effects/Niagara/Systems/Enviroment/Water/NS_WaterSplash_Medium.NS_WaterSplash_Medium" _watersurfaceposition))
    (Effects|Components|ParticleSystem|SetFloatParameter _returnvalue "WaterDepth" _waterdepth)
    (Effects|Components|ParticleSystem|SetVectorParameter _returnvalue "WaterPlaneLocation" _waterplanelocation)
    (Effects|Components|ParticleSystem|SetVectorParameter _returnvalue "WaterPlaneNormal" _waterplanenormal)
    (Effects|Components|ParticleSystem|SetFloatParameter _returnvalue "WaterBodyIndex" (Math|Conversions|ToFloat(Integer) _waterbodyidx))
    (Effects|Components|ParticleSystem|SetFloatParameter _returnvalue "EffectScale" 0.66)
    (Delivery|DoExplosion)
    (Delivery|Kill)))

(event Custom|Add%20Slap%20Effect
  (params (HitPosition :type Vector) (HitActors :type Array%20of%20Actor%20Object%20References) (HitResults :type Array%20of%20Hit%20Result%20Structures))
  ;[1 0 0 0.5] Slap Effect
  (Ability|Tasks|ApplyGameplayEffectToHitResultsAsync (Ability|GetAbilitySystemComponent (Game|GetPlayerPawn 0)) HitResults (Variables|Default|GetSlapGE) 0 1))

(fn GetAffiliationsToShowFor
  (return (Utilities|Array|MakeArray "Friendly" "Neutral" "Hostile")))

(fn OnRep_ImpactLocation
  (Components|Activation|Deactivate (Variables|Default|GetNS_SlurpBazooka_Trail)))

(fn ShieldPercent
  (params (FortPawn :type Fort%20Player%20Pawn%20Object%20Reference))
  (return (/ (FortPawn|GetShield FortPawn) (FortPawn|GetMaxShield FortPawn))))

(fn ShouldShowSoundIndicator
  (params (PlayerController :type Fort%20Player%20Controller%20Object%20Reference))
  (return true))

(fn ConstructionScript)
